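# Animation Blueprint for the CanineCronutDig backpack, adding secondary motion to its box flaps, dynamite and cords.
# A chain of spring-driven AnimDynamics nodes feeds additive bone transforms on the pack, back cord and right flap.
# The EventGraph derives pack roll from spine bend and switches pack and cord offsets when skydiving.
# It also computes sine-wobbled skydive gravity and a movement-based spring constant, though the dynamics nodes show fixed values.
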AnimBlueprint /BRCosmetics/Accessories/FORT_Backpacks/Backpack_M_MED_CanineCronutDig/Meshes/M_MED_CanineCronutDig_Pack_AnimBP
target_skeleton=&/M_MED_CanineCronutDig_Pack_Skeleton
graphs [
 AnimGraph {nodes [
   n1 = Anim Dynamics - Bone: Dyn_BoxFlad_L!(ComponentPose=Local To Component(LocalPose=Input Pose()), Alpha=1, GravityOverride:Vector="0,0,0", ExternalForce="0,0,0") -> (Pose) {
    angularDampingOverride=1
    angularSpringConstant=5000
    bAngularSpring=true
    bOverrideAngularDamping=true
    bUseGravityOverride=true
    gravityScale=0.35
    simulationSpace=RootRelative
    boundBone {boneName=Dyn_BoxFlad_L}
    componentAppliedLinearAccClamp=<5000 5000 5000>
    componentLinearAccScale=<0.55 0.55 0.55>
    componentLinearVelScale=<0.55 0.55 0.55>
    gravityOverride {z=-980}
    physicsBodyDefinitions [{
      boundBone {boneName=Dyn_BoxFlad_L}
      boxExtents=<20 30 30>
      constraintSetup {angularLimitsMax=<0 0 15>, angularLimitsMin=<0 0 -80>, angularTarget=<1 0 0>}
      localJointOffset=<4.5 0 0>
     }]
    simSpaceSettings {maxAngularAcceleration=100, maxAngularVelocity=10, simSpaceAngularAlpha=0.5}
   }
   n2 = Anim Dynamics - Bone: Dyn_BoxFlad_R!(ComponentPose=@n1, Alpha=1, GravityOverride:Vector="0,0,0", ExternalForce="0,0,0") -> (Pose) {
    angularDampingOverride=1
    angularSpringConstant=5000
    bAngularSpring=true
    bOverrideAngularDamping=true
    bUseGravityOverride=true
    gravityScale=0.35
    simulationSpace=RootRelative
    boundBone {boneName=Dyn_BoxFlad_R}
    componentAppliedLinearAccClamp=<5000 5000 5000>
    componentLinearAccScale=<0.3 0.3 0.3>
    componentLinearVelScale=<0.3 0.3 0.3>
    gravityOverride {z=-980}
    physicsBodyDefinitions [{
      boundBone {boneName=Dyn_BoxFlad_R}
      boxExtents=<20 30 30>
      constraintSetup {angularLimitsMax=<0 0 1>, angularLimitsMin=<0 0 -20>, angularTarget=<1 0 0>}
      localJointOffset=<5 0 0>
     }]
    simSpaceSettings {maxAngularAcceleration=100, maxAngularVelocity=10, simSpaceAngularAlpha=0.3}
   }
   n3 = Anim Dynamics - Bone: Dyn_dynamitecord_01!(ComponentPose=@n2, Alpha=1, GravityOverride:Vector="0,0,-980", AngularSpringConstant:float=5000, ExternalForce="0,0,0") -> (Pose) {
    angularDampingOverride=1
    angularSpringConstant=5000
    bAngularSpring=true
    bOverrideAngularDamping=true
    bUseGravityOverride=true
    gravityScale=0.35
    simulationSpace=RootRelative
    boundBone {boneName=Dyn_dynamitecord_01}
    componentAppliedLinearAccClamp=<3000 3000 3000>
    componentLinearAccScale=<0.3 0.3 0.3>
    componentLinearVelScale=<0.3 0.3 0.3>
    gravityOverride {z=-980}
    physicsBodyDefinitions [{
      boundBone {boneName=Dyn_dynamitecord_01}
      boxExtents=<20 30 30>
      constraintSetup {angularLimitsMax=<0 80 80>, angularLimitsMin=<0 -80 -80>, angularTarget=<1 0 0>}
      localJointOffset=<4 0 0>
     }]
    simSpaceSettings {maxAngularAcceleration=100, maxAngularVelocity=10, simSpaceAngularAlpha=0.2}
   }
   n4 = Anim Dynamics - Bone: Dyn_dynamitecord_02!(ComponentPose=@n3, Alpha=1, GravityOverride:Vector="0,0,-980", AngularSpringConstant:float=5000, ExternalForce="0,0,0") -> (Pose) {
    angularDampingOverride=1
    angularSpringConstant=5000
    bAngularSpring=true
    bOverrideAngularDamping=true
    bUseGravityOverride=true
    gravityScale=0.35
    simulationSpace=RootRelative
    boundBone {boneName=Dyn_dynamitecord_02}
    componentAppliedLinearAccClamp=<3000 3000 3000>
    componentLinearAccScale=<0.3 0.3 0.3>
    componentLinearVelScale=<0.3 0.3 0.3>
    gravityOverride {z=-980}
    physicsBodyDefinitions [{
      boundBone {boneName=Dyn_dynamitecord_02}
      boxExtents=<20 30 30>
      constraintSetup {angularLimitsMax=<0 80 80>, angularLimitsMin=<0 -80 -80>, angularTarget=<1 0 0>}
      localJointOffset=<4 0 0>
     }]
    simSpaceSettings {maxAngularAcceleration=100, maxAngularVelocity=10, simSpaceAngularAlpha=0.2}
   }
   n5 = Anim Dynamics - Bone: Dyn_dynamitecord_03!(ComponentPose=@n4, Alpha=1, GravityOverride:Vector="0,0,-980", AngularSpringConstant:float=5000, ExternalForce="0,0,0") -> (Pose) {
    angularDampingOverride=1
    angularSpringConstant=5000
    bAngularSpring=true
    bOverrideAngularDamping=true
    bUseGravityOverride=true
    gravityScale=0.35
    simulationSpace=RootRelative
    boundBone {boneName=Dyn_dynamitecord_03}
    componentAppliedLinearAccClamp=<3000 3000 3000>
    componentLinearAccScale=<0.3 0.3 0.3>
    componentLinearVelScale=<0.3 0.3 0.3>
    gravityOverride {z=-980}
    physicsBodyDefinitions [{
      boundBone {boneName=Dyn_dynamitecord_03}
      boxExtents=<20 30 30>
      constraintSetup {angularLimitsMax=<0 80 80>, angularLimitsMin=<0 -80 -80>, angularTarget=<1 0 0>}
      localJointOffset=<4 0 0>
     }]
    simSpaceSettings {maxAngularAcceleration=100, maxAngularVelocity=10, simSpaceAngularAlpha=0.2}
   }
   n6 = Anim Dynamics - Bone: Dyn_dynamitecord_04!(ComponentPose=@n5, Alpha=1, GravityOverride:Vector="0,0,-980", AngularSpringConstant:float=5000, ExternalForce="0,0,0") -> (Pose) {
    angularDampingOverride=1
    angularSpringConstant=5000
    bAngularSpring=true
    bOverrideAngularDamping=true
    bUseGravityOverride=true
    gravityScale=0.35
    simulationSpace=RootRelative
    boundBone {boneName=Dyn_dynamitecord_04}
    componentAppliedLinearAccClamp=<3000 3000 3000>
    componentLinearAccScale=<0.3 0.3 0.3>
    componentLinearVelScale=<0.3 0.3 0.3>
    gravityOverride {z=-980}
    physicsBodyDefinitions [{
      boundBone {boneName=Dyn_dynamitecord_04}
      boxExtents=<20 30 30>
      constraintSetup {angularLimitsMax=<0 80 80>, angularLimitsMin=<0 -80 -80>, angularTarget=<1 0 0>}
      localJointOffset=<4 0 0>
     }]
    simSpaceSettings {maxAngularAcceleration=100, maxAngularVelocity=10, simSpaceAngularAlpha=0.2}
   }
   n7 = Anim Dynamics - Bone: Dyn_dynamitecord_05!(ComponentPose=@n6, Alpha=1, GravityOverride:Vector="0,0,-980", AngularSpringConstant:float=5000, ExternalForce="0,0,0") -> (Pose) {
    angularDampingOverride=1
    angularSpringConstant=5000
    bAngularSpring=true
    bOverrideAngularDamping=true
    bUseGravityOverride=true
    gravityScale=0.35
    simulationSpace=RootRelative
    boundBone {boneName=Dyn_dynamitecord_05}
    componentAppliedLinearAccClamp=<3000 3000 3000>
    componentLinearAccScale=<0.3 0.3 0.3>
    componentLinearVelScale=<0.3 0.3 0.3>
    gravityOverride {z=-980}
    physicsBodyDefinitions [{
      boundBone {boneName=Dyn_dynamitecord_05}
      boxExtents=<20 30 30>
      constraintSetup {angularLimitsMax=<0 80 80>, angularLimitsMin=<0 -80 -80>, angularTarget=<1 0 0>}
      localJointOffset=<4 0 0>
     }]
    simSpaceSettings {maxAngularAcceleration=100, maxAngularVelocity=10, simSpaceAngularAlpha=0.2}
   }
   n8 = Anim Dynamics - Bone: Dyn_BaseDynamite_L!(ComponentPose=@n7, Alpha=1, GravityOverride:Vector="0,0,-980", ExternalForce="0,0,0") -> (Pose) {
    angularDampingOverride=1
    angularSpringConstant=10000
    bAngularSpring=true
    bOverrideAngularDamping=true
    bUseGravityOverride=true
    gravityScale=0.35
    simulationSpace=RootRelative
    boundBone {boneName=Dyn_BaseDynamite_L}
    componentAppliedLinearAccClamp=<5000 5000 5000>
    componentLinearAccScale=<0.2 0.2 0.2>
    componentLinearVelScale=<0.2 0.2 0.2>
    gravityOverride {z=-980}
    physicsBodyDefinitions [{
      boundBone {boneName=Dyn_BaseDynamite_L}
      boxExtents=<20 30 30>
      constraintSetup {angularLimitsMax=<0 20 5>, angularLimitsMin=<0 -10 -5>, angularTarget=<1 0 0>}
      localJointOffset=<4.5 0 0>
     }]
    simSpaceSettings {maxAngularAcceleration=100, maxAngularVelocity=10, simSpaceAngularAlpha=0.4}
   }
   n9 = Anim Dynamics - Bone: Dyn_BaseDynamite_R!(ComponentPose=@n8, Alpha=1, GravityOverride:Vector="0,0,-980", ExternalForce="0,0,0") -> (Pose) {
    angularDampingOverride=1
    angularSpringConstant=10000
    bAngularSpring=true
    bOverrideAngularDamping=true
    bUseGravityOverride=true
    gravityScale=0.35
    simulationSpace=RootRelative
    boundBone {boneName=Dyn_BaseDynamite_R}
    componentLinearAccScale=<0.4 0.4 0.4>
    componentLinearVelScale=<0.4 0.4 0.4>
    gravityOverride {z=-980}
    physicsBodyDefinitions [{
      boundBone {boneName=Dyn_BaseDynamite_R}
      boxExtents=<20 30 30>
      constraintSetup {angularLimitsMax=<0 40 5>, angularLimitsMin=<0 -1 -5>, angularTarget=<1 0 0>}
      localJointOffset=<4.5 0 0>
     }]
    simSpaceSettings {maxAngularAcceleration=100, maxAngularVelocity=10, simSpaceAngularAlpha=0.4}
   }
   n10 = Anim Dynamics - Bone: Dyn_BaseDynamite!(ComponentPose=@n9, Alpha=1, GravityOverride:Vector="0,0,-980", ExternalForce="0,0,0") -> (Pose) {
    angularDampingOverride=1
    angularSpringConstant=10000
    bAngularSpring=true
    bOverrideAngularDamping=true
    bUseGravityOverride=true
    gravityScale=0.35
    simulationSpace=RootRelative
    boundBone {boneName=Dyn_BaseDynamite}
    componentLinearAccScale=<0.4 0.4 0.4>
    componentLinearVelScale=<0.4 0.4 0.4>
    gravityOverride {z=-980}
    physicsBodyDefinitions [{
      boundBone {boneName=Dyn_BaseDynamite}
      boxExtents=<20 30 30>
      constraintSetup {angularLimitsMax=<0 30 5>, angularLimitsMin=<0 -5 -5>, angularTarget=<1 0 0>}
      localJointOffset=<4.5 0 0>
     }]
    simSpaceSettings {maxAngularAcceleration=100, maxAngularVelocity=10, simSpaceAngularAlpha=0.4}
   }
   n11 = Anim Dynamics - Bone: Dyn_ControlCord!(ComponentPose=@n10, Alpha=1, GravityOverride:Vector="0,0,-980", ExternalForce="0,0,0") -> (Pose) {
    angularDampingOverride=1
    angularSpringConstant=10000
    bAngularSpring=true
    bOverrideAngularDamping=true
    bUseGravityOverride=true
    gravityScale=0.35
    simulationSpace=RootRelative
    boundBone {boneName=Dyn_ControlCord}
    componentAppliedLinearAccClamp=<5000 5000 5000>
    componentLinearAccScale=<0.7 0.7 0.7>
    componentLinearVelScale=<0.7 0.7 0.7>
    gravityOverride {z=-980}
    physicsBodyDefinitions [{
      boundBone {boneName=Dyn_ControlCord}
      boxExtents=<20 30 30>
      constraintSetup {angularLimitsMax=<0 1 10>, angularLimitsMin=<0 -90 -10>, angularTarget=<1 0 0>}
      localJointOffset=<5 0 0>
     }]
    simSpaceSettings {maxAngularAcceleration=100, maxAngularVelocity=10, simSpaceAngularAlpha=0.7}
   }
   n12 = "Transform (Modify) Bone - Bone: pack"(ComponentPose=@n11, Translation=Get backpack_translation(), Rotation=Get backpack_rotate()) {
    rotationMode=BMM_Additive
    rotationSpace=BCS_BoneSpace
    translationMode=BMM_Additive
    translationSpace=BCS_BoneSpace
    boneToModify {boneName=pack}
   }
   n13 = "Transform (Modify) Bone - Bone: Base_backCord"!(ComponentPose=@n12, bAlphaBoolEnabled:bool=True, Translation="0,0,0", Rotation=Get cordRot() -> (cordRot:Rotator), Scale="1,1,1") -> (Pose) {
    alphaInputType=Bool
    rotationMode=BMM_Additive
    translationMode=BMM_Additive
    translationSpace=BCS_BoneSpace
    boneToModify {boneName=Base_backCord}
   }
   n14 = "Transform (Modify) Bone - Bone: pack"!(
    ComponentPose="Transform (Modify) Bone - Bone: Base_BoxFlad_R"(ComponentPose=@n13, Alpha=1, Translation="0,0,0", Rotation="0,-20,0", Scale="1,1,1") {rotationMode=BMM_Additive, rotationSpace=BCS_BoneSpace, boneToModify {boneName=Base_BoxFlad_R}, rotation {yaw=-20}}
    bAlphaBoolEnabled:bool
    Translation="0,-2,0"
    Rotation="0,0,-5"
    Scale
   ) -> (Pose) {
    alphaInputType=Bool
    rotationMode=BMM_Additive
    rotationSpace=BCS_BoneSpace
    translationMode=BMM_Additive
    translationSpace=BCS_BoneSpace
    boneToModify {boneName=pack}
    rotation {roll=-5}
    translation {y=-2}
   }
   n15 = "Transform (Modify) Bone - Bone: Dyn_BoxFlad_R"!(ComponentPose=@n14, bAlphaBoolEnabled:bool=True, Translation="0,0,0", Rotation="0,18,0", Scale="1,1,1") -> (Pose) {
    alphaInputType=Bool
    rotationMode=BMM_Additive
    rotationSpace=BCS_BoneSpace
    translationMode=BMM_Additive
    translationSpace=BCS_BoneSpace
    boneToModify {boneName=Dyn_BoxFlad_R}
    rotation {yaw=18}
   }
   Output Pose(Result=Component To Local(ComponentPose=@n15)) {layerGroup=None}
  ]}
 EventGraph {nodes [
   n1 = Event BlueprintUpdateAnimation()
   > n2 = Sequence!(execute) -> (then_0, then_1, then_2:Exec, then_3:Exec, then_4:Exec)
   > n3 = Branch(Condition=Get bIsSkydiveFloating())
   > Set Set Gravity(Set Gravity:Vector="0,0,800") -> (Output_Get:Vector)
   > Set SetAmplitud(SetAmplitud:Vector="350,300,400") -> (Output_Get:Vector)
   > Set SetFrequency(SetFrequency:Rotator="4,4,3") -> (Output_Get:Rotator)
   Set Set Gravity(execute=@n3.else, Set Gravity="0,0,-800")
   > Branch(Condition=Get bIsSkydiveDiveMode())
   > Set Set Gravity(Set Gravity="0,0,200")
   > Set SetAmplitud(SetAmplitud="1300,1300,600")
   > Set SetFrequency(SetFrequency="13,7,13")
   n4 = Branch(execute=@n2.then_1, Condition=Get bIsSkyDiving())
   n5 = BreakVector(InVec=Get SetAmplitud() -> (SetAmplitud:Vector))
   n6 = BreakRotator(InRot=Get SetFrequency() -> (SetFrequency:Rotator))
   n7 = GetGameTimeInSeconds()
   n8 = "Sin (Radians)"(A=float + float(A=float * float(A=@n6.Yaw, B=@n7) {is=title K2Node_PromotableOperator}, B=@n1.DeltaTimeX))
   n9 = float * float(A=@n8, B=0) {is=title K2Node_CommutativeAssociativeBinaryOperator}
   n10 = float + float(A=float * float(A=float + float(A=@n8, B=@n9) {is=title K2Node_CommutativeAssociativeBinaryOperator}, B=@n5.Z))
   n11 = "Sin (Radians)"(A=float + float(A=float * float(A=@n6, B=@n7) {is=title K2Node_PromotableOperator}, B=@n1.DeltaTimeX) {is=title K2Node_PromotableOperator})
   n12 = float * float(A=@n11, B=0) {is=title K2Node_CommutativeAssociativeBinaryOperator}
   n13 = float + float(A=float * float(A=float + float(A=@n11, B=@n12) {is=title K2Node_CommutativeAssociativeBinaryOperator}, B=@n5))
   n14 = "Sin (Radians)"(A=float + float(A=float * float(A=@n6.Pitch, B=@n7) {is=title K2Node_PromotableOperator}, B=@n1.DeltaTimeX) {is=title K2Node_PromotableOperator})
   n15 = float * float(A=@n14, B=0) {is=title K2Node_CommutativeAssociativeBinaryOperator}
   n16 = float + float(A=float * float(A=float + float(A=@n14, B=@n15) {is=title K2Node_CommutativeAssociativeBinaryOperator}, B=@n5.Y))
   > Set GravityOverride(GravityOverride=vector + vector(A=MakeVector(X=@n13, Y=@n16, Z=@n10), B=Get Set Gravity() -> (Set Gravity:Vector)))
   Set GravityOverride(execute=@n4.else, GravityOverride="0,0,-980")
   n17 = Branch(execute=@n2.then_4, Condition=Get bIsSkyDiving())
   > Set backpack_translation(backpack_translation="0,-2.2,0")
   > Set cordRot(cordRot:Rotator="0,-1.5,-7") -> (Output_Get:Rotator)
   > Set backTap_rot(backTap_rot:Rotator="0,-5,0") -> (Output_Get:Rotator)
   Set backpack_translation(execute=@n17.else, backpack_translation="0, 0, 0")
   > Set cordRot(cordRot="0,-3.4,-12")
   > Set backTap_rot(backTap_rot="0,-20,0")
   n18 = Branch(execute=@n2.then_3, Condition=NOT Boolean(A=Get bIsMoving2D()))
   > Set constant(constant:double=20000) -> (Output_Get:double)
   Set constant(execute=@n18.else, constant=5000)
   n19 = GetOwnerSkeletalMeshComponent()
   n20 = BreakRotator(InRot=BreakTransform(InTransform=GetDeltaTransformFromRefPose(self=@n19, BoneName=spine_05, BaseName=spine_01)).Rotation)
   Set backpack_rotate(execute=@n2.then_2, backpack_rotate=MakeRotator(Roll=MapRangeClamped(Value=@n20.Yaw, InRangeA=0, InRangeB=-70, OutRangeA=5, OutRangeB=15)))
   Set backpack_translation(backpack_translation=MakeVector(Y=MapRangeClamped(Value=@n20.Yaw, InRangeA=10, InRangeB=15, OutRangeA=-10, OutRangeB=10)))
   > Set Push Backpack out(Push Backpack out:bool=OR Boolean!(A=false, B=false, C:bool, D:bool=Get bIsSkydiveDiveMode()) -> (ReturnValue)) -> (Output_Get:bool)
   TryGetPawnOwner()
   PerlinNoise1D!(Value:float=@n7) -> (ReturnValue:float)
   Branch(Condition=Get bIsSkydiveDiveMode())
   Get bIsMoving2D()
   Get bIsTacticalSprinting()
  ]}
]
variables [
 backpack_rotate {type=Rotator}
 backpack_translation {type=Vector}
 backTap_rot {type=Rotator}
 constant {type=number}
 cordRot {type=Rotator}
 GravityOverride {type=Vector}
 Override World Gravity {type=Vector}
 Push Backpack out {type=boolean}
 Set Gravity {type=Vector}
 SetAmplitud {type=Vector}
 SetFrequency {type=Rotator}
]
